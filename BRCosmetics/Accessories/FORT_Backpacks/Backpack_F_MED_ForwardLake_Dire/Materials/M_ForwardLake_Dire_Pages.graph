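# Masked, two-sided character material for the pages on the ForwardLake Dire backpack cosmetic.
# Saturated TwoSidedSign picks the Diffuse texture's UV0 on front faces and UV1 on back faces, giving each side its own image.
# The SRM blue channel alone feeds Metallic, Specular and Roughness, while inverted FX red drives the opacity mask.
# These attributes pass through MF_CharacterParent_2023_PostFX; the UV tiling math and a loose scalar are unconnected.
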Material /BRCosmetics/Accessories/FORT_Backpacks/Backpack_F_MED_ForwardLake_Dire/Materials/M_ForwardLake_Dire_Pages
graph {
 properties {
  bUseMaterialAttributes=true
  bUsedWithMeshDeformer=true
  bUsedWithMorphTargets=true
  bUsedWithSkeletalMesh=true
  blendMode=BLEND_Masked
  pixelDepthOffsetMode=PDOM_Legacy
  shadingModel=MSM_FromMaterialExpression
  twoSided=true
  usedShadingModels=Default Lit
 }
 nodes [
  Add(A=TextureCoordinate(), B=Constant2Vector() {g=0.5})
  n1 = Constant2Vector() {g=2, r=3}
  n2 = Frac(None=Multiply(A=TextureCoordinate(), B=@n1))
  n3 = Divide(B=@n1)
  Add(A=Multiply(A=Multiply(A=@n2, B=@n3), B=Constant2Vector() {g=1, r=-1}), B=Constant2Vector() {g=0.5})
  n4 = TextureObjectParameter() {
   group=00_Texture Inputs
   parameterName=Diffuse
   sortPriority=0
   texture=/BRCosmetics/Accessories/FORT_Backpacks/Backpack_F_MED_ForwardLake_Dire/Textures/T_ForwardLake_Dire_Pages_D
  }
  n5 = TextureSample(UVs=TextureCoordinate() {coordinateIndex=1}, Tex=@n4) {
   samplerSource=SSM_Wrap_WorldGroupSettings
   texture=/BRCosmetics/Accessories/FORT_Backpacks/Backpack_F_MED_ForwardLake_Dire/Textures/DEMO/T_ForwardLakeDire_Pages_D
  }
  n6 = TextureSample(Tex=@n4) {
   samplerSource=SSM_Wrap_WorldGroupSettings
   texture=&T_ForwardLakeDire_Pages_D
  }
  n7 = TwoSidedSign()
  n8 = TextureSampleParameter2D() {
   desc="R = Spec , G = Roughness, B= Metal"
   group=00_Texture Inputs
   parameterName=SRM
   samplerType=SAMPLERTYPE_Masks
   sortPriority=1
   texture=/BRCosmetics/Accessories/FORT_Backpacks/Backpack_F_MED_ForwardLake_Dire/Textures/DEMO/T_ForwardLakeDire_Pages_SRM
  }
  n9 = TextureSampleParameter2D() {
   group=00_Texture Inputs
   parameterName=Normals
   samplerType=SAMPLERTYPE_Normal
   sortPriority=2
   texture=&FlatNormalBlue
  }
  n10 = TextureSample() {
   samplerType=SAMPLERTYPE_Masks
   texture=/BRCosmetics/Accessories/FORT_Backpacks/Backpack_F_MED_ForwardLake_Dire/Textures/T_ForwardLake_Dire_Pages_FX
  }
  n11 = SetMaterialAttributes!(
   MaterialAttributes
   Base Color=LinearInterpolate(A=@n5.RGB, B=@n6.RGB, Alpha=Saturate(None=@n7))
   Metallic=@n8.B
   Specular=@n8.B
   Roughness=@n8.B
   Normal=@n9.RGB
   Opacity Mask=OneMinus(None=@n10.R)
  ) {attributeSetTypes [
    #1
    #2
    #3
    #4
    #5
    #6
   ]}
  MaterialFunctionCall(Material=@n11, M=Constant(), Fresnel=Fresnel() {baseReflectFraction=0, exponent=1}) {materialFunction=&MF_CharacterParent_2023_PostFX}
  ScalarParameter() {defaultValue=0.7145}
 ]
}
