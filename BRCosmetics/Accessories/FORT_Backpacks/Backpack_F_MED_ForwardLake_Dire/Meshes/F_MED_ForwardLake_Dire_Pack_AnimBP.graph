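# Animation Blueprint for the ForwardLake Dire backpack, playing a chest-open reaction with a physics-driven ribbon.
# A front-end bool blends two Closed/Open state machines: the in-game one triggers on the aie.ChestOpened input event, the lobby one on delayed, self-repeating flags when the part modifier enables reactivity.
# Trigger flags clear after each evaluation.
# An AnimDynamics chain then simulates the ribbon; skydive-dependent gravity is stored, but the node's override pin shows zero.
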AnimBlueprint /BRCosmetics/Accessories/FORT_Backpacks/Backpack_F_MED_ForwardLake_Dire/Meshes/F_MED_ForwardLake_Dire_Pack_AnimBP
target_skeleton=&/F_MED_ForwardLake_Dire_Pack_Skeleton
graphs [
 AnimGraph {nodes [
   AnimGraphNode_StateMachine_0 = Reactivity() {is=title AnimGraphNode_StateMachine}
   AnimGraphNode_StateMachine_1 = FE_Reactivity() {is=title AnimGraphNode_StateMachine}
   n1 = Get bIsInFrontEnd()
   n2 = "Anim Dynamics - Chain: JNT_Ribbon_01 -> JNT_Ribbon_04"!(
    ComponentPose=Local To Component(LocalPose=Blend Poses by bool(BlendPose_0=@AnimGraphNode_StateMachine_1, BlendPose_1=@AnimGraphNode_StateMachine_0, bActiveValue=@n1) {blendType=Linear})
    Alpha=1
    GravityOverride:Vector="0,0,0"
   ) -> (Pose) {
    angularBiasOverride=0.999
    angularDampingOverride=0.9999
    angularSpringConstant=50
    bChain=true
    bOverrideAngularDamping=true
    bUseGravityOverride=true
    bUsePlanarLimit=false
    numSolverIterationsPostUpdate=4
    numSolverIterationsPreUpdate=16
    simulationSpace=RootRelative
    alphaBoolBlend {blendOption=Linear}
    boundBone {boneName=JNT_Ribbon_01}
    chainEnd {boneName=JNT_Ribbon_04}
    componentAppliedLinearAccClamp=<1000 1000 1000>
    componentLinearAccScale=<0.4 0.9 0.4>
    componentLinearVelScale=<0.4 0.9 0.4>
    physicsBodyDefinitions [
     {
      sphereCollisionRadius=1
      boundBone {boneName=JNT_Ribbon_01}
      boxExtents=<22 22 22>
      constraintSetup {angularLimitsMax=<0 45 45>, angularLimitsMin=<0 -45 -45>, angularTarget=<1 0 0>}
      localJointOffset=<2.5 0 0>
     }
     {
      sphereCollisionRadius=1
      boundBone {boneName=JNT_Ribbon_02}
      boxExtents=<19 19 19>
      constraintSetup {angularLimitsMax=<0 45 45>, angularLimitsMin=<0 -45 -45>, angularTarget=<1 0 0>}
      localJointOffset=<2.5 0 0>
     }
     {
      sphereCollisionRadius=1
      boundBone {boneName=JNT_Ribbon_03}
      boxExtents=<16 16 16>
      constraintSetup {angularLimitsMax=<0 45 45>, angularLimitsMin=<0 -45 -45>, angularTarget=<1 0 0>}
      localJointOffset=<3 0 0>
     }
     {
      sphereCollisionRadius=1
      boundBone {boneName=JNT_Ribbon_04}
      boxExtents=<11 11 11>
      constraintSetup {angularLimitsMax=<0 45 45>, angularLimitsMin=<0 -45 -45>, angularTarget=<1 0 0>}
      localJointOffset=<5.5 0 0>
     }
    ]
   }
   Output Pose(Result=Component To Local(ComponentPose=@n2)) {layerGroup=None}
  ]}
 Reactivity {
  entry=Idle
  path=AnimGraph.AnimGraphNode_StateMachine_0.Reactivity
  ids=id
  nodes [
   AnimStateNode_0 = Idle() {stateEntered {guid=#1}}
   AnimStateNode_5 = OpenChest() {is=title AnimStateNode, stateEntered {guid=#2, notifyName=OpenChestEntered}}
   AnimStateTransitionNode_7 = AnimStateTransitionNode()
   AnimStateTransitionNode_8 = AnimStateTransitionNode() {bAutomaticRuleBasedOnSequencePlayerInState=true}
  ]
  links [
   {
    dst=OpenChest
    rule=AnimGraph.AnimGraphNode_StateMachine_0.Reactivity.AnimStateTransitionNode_7.Transition
    src=Idle
    via=AnimStateTransitionNode_7
   }
   {
    dst=Idle
    rule=AnimGraph.AnimGraphNode_StateMachine_0.Reactivity.AnimStateTransitionNode_8.Transition
    src=OpenChest
    via=AnimStateTransitionNode_8
   }
  ]
 }
 Idle {
  path=AnimGraph.AnimGraphNode_StateMachine_0.Reactivity.AnimStateNode_0.Idle
  nodes [
   n1 = Sequence Player 'ForwardLake_Dire_Closed_Pack'() {bLoopAnimation=false, sequence=/BRCosmetics/Accessories/FORT_Backpacks/Backpack_F_MED_ForwardLake_Dire/Anim/ForwardLake_Dire_Closed_Pack}
   Output Animation Pose(Result=@n1)
  ]
 }
 OpenChest {
  path=AnimGraph.AnimGraphNode_StateMachine_0.Reactivity.AnimStateNode_5.OpenChest
  nodes [
   n1 = Sequence Player 'ForwardLake_Dire_Open_Pack'() {bLoopAnimation=false, sequence=/BRCosmetics/Accessories/FORT_Backpacks/Backpack_F_MED_ForwardLake_Dire/Anim/ForwardLake_Dire_Open_Pack}
   Output Animation Pose(Result=@n1)
  ]
 }
 Transition {
  path=AnimGraph.AnimGraphNode_StateMachine_0.Reactivity.AnimStateTransitionNode_7.Transition
  nodes [Result(bCanEnterTransition=Get bChestOpened() -> (bChestOpened:bool))]
 }
 Transition {path=AnimGraph.AnimGraphNode_StateMachine_0.Reactivity.AnimStateTransitionNode_8.Transition, nodes [Result()]}
 FE_Reactivity {
  entry=FE_Idle
  path=AnimGraph.AnimGraphNode_StateMachine_1.FE_Reactivity
  ids=id
  nodes [
   AnimStateNode_0 = FE_Idle() {is=title AnimStateNode, stateEntered {guid=#1}}
   AnimStateNode_5 = FE_OpenChest() {is=title AnimStateNode, stateEntered {guid=#3, notifyName=FE_OpenChestEntered}}
   AnimStateTransitionNode_7 = AnimStateTransitionNode()
   AnimStateTransitionNode_8 = AnimStateTransitionNode() {bAutomaticRuleBasedOnSequencePlayerInState=true}
  ]
  links [
   {
    dst=FE_OpenChest
    rule=AnimGraph.AnimGraphNode_StateMachine_1.FE_Reactivity.AnimStateTransitionNode_7.Transition
    src=FE_Idle
    via=AnimStateTransitionNode_7
   }
   {
    dst=FE_Idle
    rule=AnimGraph.AnimGraphNode_StateMachine_1.FE_Reactivity.AnimStateTransitionNode_8.Transition
    src=FE_OpenChest
    via=AnimStateTransitionNode_8
   }
  ]
 }
 FE_Idle {
  path=AnimGraph.AnimGraphNode_StateMachine_1.FE_Reactivity.AnimStateNode_0.FE_Idle
  nodes [
   n1 = Sequence Player 'ForwardLake_Dire_Closed_Pack'() {bLoopAnimation=false, sequence=&ForwardLake_Dire_Closed_Pack}
   Output Animation Pose(Result=@n1)
  ]
 }
 FE_OpenChest {
  path=AnimGraph.AnimGraphNode_StateMachine_1.FE_Reactivity.AnimStateNode_5.FE_OpenChest
  nodes [
   n1 = Sequence Player 'ForwardLake_Dire_Open_Pack'() {bLoopAnimation=false, sequence=&ForwardLake_Dire_Open_Pack}
   Output Animation Pose(Result=@n1)
  ]
 }
 Transition {
  path=AnimGraph.AnimGraphNode_StateMachine_1.FE_Reactivity.AnimStateTransitionNode_7.Transition
  nodes [Result(bCanEnterTransition=Get Play_FE_Reactivity() -> (Play_FE_Reactivity:bool))]
 }
 Transition {path=AnimGraph.AnimGraphNode_StateMachine_1.FE_Reactivity.AnimStateTransitionNode_8.Transition, nodes [Result()]}
 EventGraph {nodes [
   Event BlueprintUpdateAnimation()
   n1 = Get bIsSkydiveDiveMode()
   n2 = Get bIsSkydiveFloating()
   > Set Gravity Override(Gravity Override=SelectVector(A="0,0,500", B="0,0,-980", bPickA=OR Boolean(A=@n1, B=@n2)))
   Event OnCharacterCustomizationLoaded()
   > n3 = Cast To FortPlayerPawn(Object=TryGetFortPawn())
   > n4 = Cast To B_Athena_PartModifier_Backpack_ForwardLake_Dire!(
    execute
    Object=GetModifierForPartType!(self:Fort Player Pawn Object Reference=@n3, PartType:EFortCustomPartType Enum=Backpack) -> (ReturnValue:Custom Character Part Modifier Object Reference)
   ) -> (then, CastFailed, AsB Athena Part Modifier Backpack Forward Lake Dire:B Athena Part Modifier Backpack Forward Lake Dire Object Reference) {is=title K2Node_DynamicCast}
   > n5 = Branch(Condition=Get ReactivityEnabled!(self:B Athena Part Modifier Backpack Forward Lake Dire Object Reference=@n4) -> (ReactivityEnabled:bool))
   > Do Once(Reset=@n7.then)
   > n6 = Delay(Duration=4)
   Set Play_FE_Reactivity(execute=[@n8.then, @n6.then], Play_FE_Reactivity:bool=true) -> (Output_Get:bool)
   n7 = Set Play_FE_Reactivity(execute=@n5.else, Play_FE_Reactivity=false)
   Event BlueprintPostEvaluateAnimation() {is=title K2Node_Event}
   > Set bChestOpened(bChestOpened:bool=false) -> (Output_Get:bool)
   > Set Play_FE_Reactivity(Play_FE_Reactivity=false)
   AnimNotify_FE_OpenChestEntered() {is=title K2Node_Event}
   > Set Play_FE_Reactivity(Play_FE_Reactivity=false)
   > n8 = Delay(Duration=10)
   n9 = Event BlueprintProcessAnimInputEvent()
   > Branch(Condition=IsNamed(self=@n9.AnimInputEvent, InEventName=aie.ChestOpened))
   > Set bChestOpened(bChestOpened=true)
   TryGetPawnOwner()
  ]}
]
variables [
 bChestOpened {type=boolean}
 Gravity Override {type=Vector, default=<0 0 -980>}
 Play_FE_Reactivity {type=boolean}
]
